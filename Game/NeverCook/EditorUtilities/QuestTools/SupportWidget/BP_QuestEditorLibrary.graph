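# Shared editor-utility library for the quest authoring tool, holding its common state and broadcasting changes through dispatchers.
# It creates quest assets from data-table templates and duplicates quests by incrementing trailing numbers, copying release version and regenerating objective backend names.
# Loading a quest applies templated properties, appending gameplay tag containers and copying others via inline Python.
# Save-game hooks persist UI preferences, and a helper overlays widgets on the root canvas.
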EditorUtilityBlueprint /Game/NeverCook/EditorUtilities/QuestTools/SupportWidget/BP_QuestEditorLibrary
graphs [
 AddWidgetToRoot {nodes [
   n1 = AddWidgetToRoot!() -> (then, WidgetType:User Widget Class Reference) {is=title K2Node_FunctionEntry}
   n2 = Get CanvasRoot() -> (CanvasRoot:Canvas Panel Object Reference)
   > n3 = Is Valid(InputObject=@n2)
   PrintString(execute=@n3.Is Not Valid, InString="AddWidgetToRoot failed, there is no root canvas panel defined. Use SetCanvasRoot before calling this!")
   > Return Node!(execute, WidgetInstance:User Widget Object Reference)
   n4 = Create Widget!(
    execute:Exec=@n3.Is Valid
    WorldContextObject:Object Reference=@n2
    Class:User Widget Class Reference=@n1
    OwningPlayer:Player Controller Object Reference=GetOwningPlayer(self=@n2)
   ) -> (then:Exec, ReturnValue:User Widget Object Reference)
   > n5 = AddChildToCanvas!(execute, self:Canvas Panel Object Reference=Get CanvasRoot(), Content:Widget Object Reference=@n4) -> (then, ReturnValue:Canvas Panel Slot Object Reference)
   > SetOffsets!(execute, self:Canvas Panel Slot Object Reference=@n5, InOffset:Margin Structure=Make Margin(Left=0, Top=0, Right=0, Bottom=0)) -> (then)
   > SetAnchors!(
    execute
    self:Canvas Panel Slot Object Reference=@n5
    InAnchors:Anchors Structure=Make Anchors!(Minimum:Vector 2D Structure=(X=0,Y=0), Maximum:Vector 2D Structure=(X=1,Y=1)) -> (Anchors:Anchors Structure) {is=title K2Node_MakeStruct}
   ) -> (then)
   > SetZOrder!(execute, self:Canvas Panel Slot Object Reference=@n5, InZOrder:int=9999) -> (then)
   > Return Node(WidgetInstance=@n4)
  ]}
 AttemptToIncrementFileName {nodes [
   AttemptToIncrementFileName() {is=title K2Node_FunctionEntry}
   > Set WorkingFileName(WorkingFileName:String=Break Quest Editor Common Data!(QuestEditorCommonData:"Quest Editor Common Data Structure (by ref)"=Get CommonData() -> (CommonData:Quest Editor Common Data Structure)) -> (
     ActiveObject_28_#1:Object Reference
     FileName_20_#2:String
     WidgetType_27_#3:QuestToolWidgetType Enum
     QuestTemplateTypeName_39_#4:Name
    ) {is=title K2Node_BreakStruct}.FileName_20_#2) -> (Output_Get:String)
   n1 = Get WorkingFileName() -> (WorkingFileName:String)
   > n2 = Branch(Condition=IsNumeric!(SourceString:String=Right!(SourceString:String=@n1, Count:int=1) -> (ReturnValue:String)) -> (ReturnValue:bool))
   n3 = Get WorkingFileName()
   n4 = Get NumericalCharacters() -> (NumericalCharacters:int=0)
   > n5 = While Loop!(execute:Exec, Condition:bool=IsNumeric(SourceString=Right(SourceString=@n3, Count=@n4))) -> (LoopBody:Exec, Completed:Exec) {is=title K2Node_MacroInstance}
   Decrement Int(" "=@n5.Completed, Value=Get NumericalCharacters() -> (NumericalCharacters=0))
   n6 = Get WorkingFileName()
   n7 = Get NumericalCharacters() -> (NumericalCharacters=0)
   > n8 = Increment Int(Value=String To Integer(InString=Right(SourceString=@n6, Count=@n7)))
   n9 = LeftChop(SourceString=Get WorkingFileName(), Count=Get NumericalCharacters() -> (NumericalCharacters=0))
   n10 = "To String (Integer)"(InInt=@n8)
   > n11 = Set WorkingFileName(WorkingFileName=SelectString(A=Append!(A=@n9, B=0, C:String=@n10) -> (ReturnValue), B=Append!(A=@n9, B=@n10) -> (ReturnValue), bPickA="integer <= integer"(A=@n8, B=9)))
   > Return Node!(execute, Success:bool=true, NewFileName:String=@n11)
   Increment Int(" "=@n5.LoopBody, Value=Get NumericalCharacters() -> (NumericalCharacters=0))
   Show Message Dialog(
    execute=@n2.else
    Title=MakeLiteralText(Value="LOCTEXT(\"Quest name doesn\\'t end in a number\")")
    Message=MakeLiteralText(Value="LOCTEXT(\"To make the next in the series, close the quest editor and rename the quest to end with a number then right click and reopen it in the editor. e.g, daily_quest_03\")")
    MessageType=Ok
    DefaultValue=No
    MessageCategory=Warning
   )
   > Return Node(Success=false)
  ]}
 CreateNewQuestAsset {nodes [
   n1 = CreateNewQuestAsset!() -> (then, FileName:String, Folder:Directory Path Structure) {is=title K2Node_FunctionEntry}
   n2 = GetSelectedTemplateObjectCurrentStyle!(self) -> (Asset Object:Fort Quest Item Definition Object Reference, Success:bool)
   > n3 = Branch(Condition=@n2.Success)
   > n4 = Set QuestItemClassReference(QuestItemClassReference:Fort Quest Item Definition Class Reference=Get Class(Object=@n2) -> (ReturnValue:Fort Quest Item Definition Class Reference)) -> (Output_Get:Fort Quest Item Definition Class Reference)
   n5 = GetAssetTools(execute=[@n4.then, @n3.else])
   > n6 = CreateAsset(
    self=@n5
    AssetName=@n1
    PackagePath=Break Directory Path!(DirectoryPath:"Directory Path Structure (by ref)"=@n1.Folder) -> (Path:String) {is=title K2Node_BreakStruct}
    AssetClass=Get QuestItemClassReference() -> (QuestItemClassReference:Fort Quest Item Definition Class Reference)
   )
   > n7 = Cast To FortQuestItemDefinition!(execute, Object=@n6) -> (then, CastFailed, AsFort Quest Item Definition:Fort Quest Item Definition Object Reference) {is=title K2Node_DynamicCast}
   n8 = Break Quest Editor Common Data!(QuestEditorCommonData=GetCommonData!(self) -> (CommonData:Quest Editor Common Data Structure)) -> (CurrentSeasonStyle_9_#5:String, QuestTemplateTypeName_39_#4)
   > Set QuestToolTemplate!(execute, QuestToolTemplate:String=@n8, self:Fort Quest Item Definition Object Reference=@n7) -> (then, Output_Get:String)
   > PrintString(InString=Append(A="Save Asset: ", B=@n8))
   > Return Node!(execute, New Quest Object:Object Reference=@n6)
  ]}
 DuplicateCurrentQuest {nodes [
   DuplicateCurrentQuest() {is=title K2Node_FunctionEntry}
   > n1 = AttemptToIncrementFileName!(execute, self) -> (then, Success:bool, NewFileName:String) {is=title K2Node_CallFunction}
   > Branch(Condition=@n1)
   > n2 = GetAssetTools()
   n3 = Break Quest Editor Common Data!(QuestEditorCommonData=Get CommonData()) -> (
    ActiveObject_28_#1:Object Reference
    Folder_24_#6:Directory Path Structure
    WidgetType_27_#3:QuestToolWidgetType Enum
    QuestTemplateTypeName_39_#4
   )
   > n4 = DuplicateAsset!(
    execute
    self:Asset Tools Interface=@n2
    AssetName:String=@n1.NewFileName
    PackagePath:String=Break Directory Path(DirectoryPath=@n3.Folder_24_#6)
    OriginalObject:Object Reference=@n3
   ) -> (then, ReturnValue:Object Reference)
   > n5 = Is Valid(InputObject=@n4)
   > n6 = Cast To FortQuestItemDefinition(Object=@n4)
   > n7 = Set DuplicatedQuest(DuplicatedQuest:Fort Quest Item Definition Object Reference=@n6) -> (Output_Get:Fort Quest Item Definition Object Reference)
   n8 = Break Quest Editor Common Data(QuestEditorCommonData=Get CommonData())
   > n9 = Cast To FortQuestItemDefinition(Object=@n8)
   > ForceScanForNewAssets!(
    execute
    self
    FolderPath:String=Break Directory Path(DirectoryPath=@n8.Folder_24_#6)
    NewFileName:String=@n1.NewFileName
   ) -> (then)
   > CopyFortReleaseVersion!(execute, ItemDefinition:Fort Item Definition Object Reference=@n7, ItemDefinitionToCopyFrom:Fort Item Definition Object Reference=@n9) -> (then, ReturnValue:bool)
   > Set members in Quest Editor Common Data!(
    execute:Exec
    StructRef:"Quest Editor Common Data Structure (by ref)"=Get CommonData()
    FileName_20_#2:String=@n1.NewFileName
   ) -> (then:Exec, StructOut:"Quest Editor Common Data Structure (by ref)") {is=title K2Node_SetFieldsInStruct}
   > n10 = For Each Loop(Array:Array of Fort Mcp Quest Objective Info Structures=GetObjectives!(self:Fort Quest Item Definition Object Reference=Get DuplicatedQuest() -> (DuplicatedQuest:Fort Quest Item Definition Object Reference)) -> (ReturnValue:Array of Fort Mcp Quest Objective Info Structures)) -> (Array Element:Fort Mcp Quest Objective Info Structure)
   Return Node(execute=@n10.Completed, New Quest Object=Get DuplicatedQuest())
   > GenerateUniqueBackendNameForObjective!(
    execute
    self:Fort Quest Item Definition Object Reference=Get DuplicatedQuest()
    ObjectiveIndex:int=@n10.Array Index
    Reason:EQuestEditorGenerateBackendNameReason Enum=ForceRebuild
   ) -> (then, ReturnValue:bool)
   Return Node(execute=[@n6.CastFailed, @n5.Is Not Valid], New Quest Object=@n4)
  ]}
 EventGraph {links [], nodes []}
 ForceScanForNewAssets {nodes [
   n1 = ForceScanForNewAssets!() -> (then, FolderPath:String, NewFileName:String) {is=title K2Node_FunctionEntry}
   > ScanFilesSynchronous!(
    execute
    self:Asset Registry Interface=GetAssetRegistry()
    InFilePaths:Array of Strings=Make Array("[0]":String=Append!(A=@n1, B=/, C:String=@n1.NewFileName) -> (ReturnValue)) -> (Array:Array of Strings)
    bForceRescan:bool=true
   ) -> (then)
  ]}
 Get Template Data Table {nodes [
   Get Template Data Table() {is=title K2Node_FunctionEntry}
   > Return Node!(execute, NewParam:Data Table Object Reference=Get Templates() -> (Templates:Data Table Object Reference))
  ]}
 GetCommonData {nodes [
   GetCommonData() {is=title K2Node_FunctionEntry}
   > Return Node!(execute, CommonData:Quest Editor Common Data Structure=Get CommonData())
  ]}
 GetSavedFilePath {nodes [
   GetSavedFilePath() {is=title K2Node_FunctionEntry}
   > Return Node!(execute, SavedFilePath:String=Get SavedFilePickerPath() -> (SavedFilePickerPath:String))
  ]}
 GetSelectedTemplateObjectCurrentStyle {nodes [
   GetSelectedTemplateObjectCurrentStyle() {is=title K2Node_FunctionEntry}
   n1 = GetSelectedTemplateObjectCustomStyle!(
    self
    Template Name:String=Break Quest Editor Common Data!(QuestEditorCommonData=Get CommonData()) -> (CurrentSeasonStyle_9_#5:String, QuestTemplateTypeName_39_#4)
   ) -> (QuestTemplate:Fort Quest Item Definition Object Reference, Success:bool)
   > Return Node!(execute, Asset Object:Fort Quest Item Definition Object Reference=@n1, Success:bool=@n1.Success)
  ]}
 GetSelectedTemplateObjectCustomStyle {nodes [
   n1 = GetSelectedTemplateObjectCustomStyle!() -> (then, Template Name:String) {is=title K2Node_FunctionEntry}
   n2 = String To Name(InString=@n1)
   > n3 = Branch(Condition="Not Equal (Name)"(A=@n2) -> (ReturnValue=~))
   > n4 = Branch(Condition="Not Equal (Name)"(A=@n2, B=Unset) -> (ReturnValue=~))
   > n5 = LoadQuestTemplateFromTableByRowName!(execute, QuestTemplatesTable:Data Table Object Reference=Get Templates(), TemplateRowName:Name=@n2) -> (then, ReturnValue:Fort Quest Item Definition Object Reference)
   > Return Node!(execute, QuestTemplate:Fort Quest Item Definition Object Reference=@n5, Success=IsValid(Object=@n5))
   Return Node(execute=@n4.else, Success=false)
   PrintString(execute=@n3.else, InString=Quest Editor Library tried to load a style that was None.)
   > Return Node(Success=false)
  ]}
 GetShowingCommonTags {nodes [
   GetShowingCommonTags() {is=title K2Node_FunctionEntry}
   > Return Node!(execute, Enabled:bool=Get ShowingCommonTags() -> (ShowingCommonTags:bool))
  ]}
 Init Common Data {nodes [
   n1 = Init Common Data!() -> (
    then
    TemplatedPropertyNames:Array of Names
    WidgetType:"QuestToolWidgetType Enum (by ref)"=NewEnumerator0
    IsDesignTime:bool=true
    QuestTemplateTypeName:Name
   ) {is=title K2Node_FunctionEntry}
   > SaveGame_Init(self:BP Editor Object Save Game Base Object Reference)
   > Set members in Quest Editor Common Data!(
    execute
    StructRef=Get CommonData()
    CurrentSeasonStyle_9_#5:String=Unset
    TemplatedPropertyNames_13_#7:Array of Names=@n1
    WidgetType_27_#3:QuestToolWidgetType Enum=@n1.WidgetType
    QuestTemplateTypeName_39_#4:Name=@n1.QuestTemplateTypeName
   ) -> (then, StructOut)
  ]}
 LoadSelectedAsset {nodes [
   LoadSelectedAsset() {is=title K2Node_FunctionEntry}
   > n1 = GetSelectedAssets()
   > n2 = Cast To FortQuestItemDefinition(Object="Get (a copy)"(Array:Array of Object References=@n1) -> (Output:Object Reference))
   > SetActiveObject!(execute, self, ActiveObject:Object Reference=@n2) -> (then)
   n3 = GetSelectedTemplateObjectCustomStyle(Template Name=Get QuestToolTemplate!(self:Fort Quest Item Definition Object Reference=@n2) -> (QuestToolTemplate:String)) {is=title K2Node_CallFunction}
   > Branch(Condition=@n3.Success)
   > WriteQuestPropertiesFromTemplate!(execute, self, Template To Use:Object Reference=@n3, Quest Object To Use:Object Reference=@n2) -> (then)
   Return Node!(execute=@n2.CastFailed, Success:bool=false)
   n4 = Break Quest Editor Common Data!(QuestEditorCommonData=Get CommonData()) -> (
    ActiveObject_28_#1:Object Reference
    WidgetType_27_#3:QuestToolWidgetType Enum
    QuestTemplateTypeName_39_#4
   )
   Switch on QuestToolWidgetType!(execute, Selection:QuestToolWidgetType Enum=@n4.WidgetType_27_#3) -> (NewEnumerator0:Exec, NewEnumerator1:Exec, NewEnumerator2:Exec, NewEnumerator3:Exec) {is=title K2Node_SwitchEnum}
   > n5 = Cast To FortQuestItemDefinition(Object=@n4)
   n6 = GetSelectedTemplateObjectCustomStyle(Template Name=Get QuestToolTemplate(self=@n5))
   > n7 = Branch(Condition=@n6.Success)
   > n8 = GetSelectedAssets()
   n9 = "Get (a copy)"(Array=@n8)
   > SetActiveObject(ActiveObject=@n9)
   > WriteQuestPropertiesFromTemplate(Template To Use=@n6, Quest Object To Use=@n9)
   > Return Node(Success=true)
   PrintString(execute=@n7.else, InString="GetSelectedTemplateObjectCurrentStyle failed to find the Current Season Style, so we can't load the quest!")
   > n10 = GetSelectedAssets()
   > SetActiveObject(ActiveObject="Get (a copy)"(Array=@n10))
   > Return Node(Success=true)
  ]}
 OnActiveObjectChanged {nodes [OnActiveObjectChanged!() -> (then, ActiveObject:Object Reference, FileName:String, Folder:Directory Path Structure) {is=title K2Node_FunctionEntry}]}
 OnBannerMessage {nodes [OnBannerMessage!() -> (then, Message:Text) {is=title K2Node_FunctionEntry}]}
 OnCopiedObjectiveIndexSet {nodes [OnCopiedObjectiveIndexSet() {is=title K2Node_FunctionEntry}]}
 OnSavedFilePickerPathChanged {nodes [OnSavedFilePickerPathChanged!() -> (then, FilePickerPath:String) {is=title K2Node_FunctionEntry}]}
 OnShowCommonTagsChanged {nodes [OnShowCommonTagsChanged!() -> (then, Enabled:bool=~) {is=title K2Node_FunctionEntry}]}
 SaveGame_PostLoad {nodes [
   n1 = SaveGame_PostLoad!() -> (then, SaveGameExisted:bool, SaveGameObject:Save Game Object Reference) {is=title K2Node_FunctionEntry}
   > n2 = Cast To BP_QuestEditor_SaveData!(execute, Object=@n1.SaveGameObject) -> (then, CastFailed, AsBP Quest Editor Save Data:BP Quest Editor Save Data Object Reference) {is=title K2Node_DynamicCast}
   > n3 = Set ShowingCommonTags(ShowingCommonTags:bool=Get ShowingCommonTags!(self:BP Quest Editor Save Data Object Reference=@n2) -> (ShowingCommonTags)) -> (Output_Get:bool)
   > Call On Show Common Tags Changed!(execute:Exec, self:BP Quest Editor Library Object Reference, Enabled:bool=@n3) -> (then:Exec) {is=title K2Node_CallDelegate}
   n4 = Get FilePickerSavedPath!(self:BP Quest Editor Save Data Object Reference=@n2) -> (FilePickerSavedPath:String)
   > Set SavedFilePickerPath(SavedFilePickerPath:String=@n4) -> (Output_Get:String)
   > Call On Saved File Picker Path Changed!(execute:Exec, self:BP Quest Editor Library Object Reference, FilePickerPath:String=@n4) -> (then:Exec) {is=title K2Node_CallDelegate}
   > Set ShowBrowseToAssetWarning(ShowBrowseToAssetWarning:bool=Get ShowBrowseToAssetWarning!(self:BP Quest Editor Save Data Object Reference=@n2) -> (ShowBrowseToAssetWarning:bool)) -> (Output_Get:bool)
  ]}
 SaveGame_PrepareToSave {nodes [
   n1 = SaveGame_PrepareToSave!() -> (then, SaveGameObject:Save Game Object Reference) {is=title K2Node_FunctionEntry}
   > n2 = Cast To BP_QuestEditor_SaveData(Object=@n1)
   > Set ShowingCommonTags!(execute, ShowingCommonTags=Get ShowingCommonTags!() -> (ShowingCommonTags), self:BP Quest Editor Save Data Object Reference=@n2) -> (then, Output_Get)
   > Set FilePickerSavedPath!(execute, FilePickerSavedPath:String=Get SavedFilePickerPath(), self:BP Quest Editor Save Data Object Reference=@n2) -> (then, Output_Get:String)
   > Set ShowBrowseToAssetWarning!(
    execute
    ShowBrowseToAssetWarning=Get ShowBrowseToAssetWarning!() -> (ShowBrowseToAssetWarning)
    self:BP Quest Editor Save Data Object Reference=@n2
   ) -> (then, Output_Get)
  ]}
 SetActiveObject {nodes [
   n1 = SetActiveObject!() -> (then, ActiveObject:Object Reference) {is=title K2Node_FunctionEntry}
   > PrintString(InString=Append!(A="Active Object Set: ", B=GetDisplayName(Object=@n1)) -> (ReturnValue))
   n2 = Break Asset Data!(AssetData=CreateAssetData(InAsset=@n1, bAllowBlueprintClass=true)) -> (PackagePath, AssetName, AssetClassPath)
   n3 = Make Directory Path!(Path:String="To String (Name)"(InName=@n2)) -> (DirectoryPath:Directory Path Structure) {is=title K2Node_MakeStruct}
   n4 = "To String (Name)"(InName=@n2.AssetName)
   > Set members in Quest Editor Common Data!(
    execute
    StructRef=Get CommonData()
    ActiveObject_28_#1:Object Reference=@n1
    FileName_20_#2:String=@n4
    Folder_24_#6:Directory Path Structure=@n3
   ) -> (then, StructOut)
   > Call On Active Object Changed!(
    execute:Exec
    self:BP Quest Editor Library Object Reference
    ActiveObject:Object Reference=@n1
    FileName:String=@n4
    Folder:Directory Path Structure=@n3
   ) -> (then:Exec) {is=title K2Node_CallDelegate}
  ]}
 SetCanvasRoot {nodes [
   n1 = SetCanvasRoot!() -> (then, CanvasRoot:Canvas Panel Object Reference) {is=title K2Node_FunctionEntry}
   > Set CanvasRoot(CanvasRoot:Canvas Panel Object Reference=@n1) -> (Output_Get:Canvas Panel Object Reference)
  ]}
 SetCopiedObjectiveIndex {nodes [
   SetCopiedObjectiveIndex() {is=title K2Node_FunctionEntry}
   > Call On Copied Objective Index Set!(execute:Exec, self:BP Quest Editor Library Object Reference) -> (then:Exec) {is=title K2Node_CallDelegate}
  ]}
 SetCurrentSeasonStyle {nodes [
   n1 = SetCurrentSeasonStyle!() -> (then, CurrentSeasonStyle:String) {is=title K2Node_FunctionEntry}
   > Set members in Quest Editor Common Data!(execute, StructRef=Get CommonData(), CurrentSeasonStyle_9_#5:String=@n1) -> (then, StructOut)
  ]}
 SetLoadedAsset {nodes [
   n1 = SetLoadedAsset!() -> (then, LoadedObject:Object Reference) {is=title K2Node_FunctionEntry}
   > Switch on QuestToolWidgetType(Selection=Break Quest Editor Common Data!(QuestEditorCommonData=Get CommonData()) -> (WidgetType_27_#3, QuestTemplateTypeName_39_#4))
   > SetActiveObject(ActiveObject=@n1) {is=title K2Node_CallFunction}
   n2 = GetSelectedTemplateObjectCurrentStyle() {is=title K2Node_CallFunction}
   > n3 = Branch(Condition=@n2.Success)
   > WriteQuestPropertiesFromTemplate(Template To Use=@n2, Quest Object To Use=@n1)
   > Return Node(Success=true)
   PrintString(execute=@n3.else, InString="GetSelectedTemplateObjectCurrentStyle failed to find the Current Season Style, so we can't load the quest!")
   > Return Node(Success=false)
  ]}
 SetSavedFilePath {nodes [
   n1 = SetSavedFilePath!() -> (then, InSavedFilePath:String) {is=title K2Node_FunctionEntry}
   > Set SavedFilePickerPath(SavedFilePickerPath=@n1)
   > SaveGame_Save(self:BP Editor Object Save Game Base Object Reference)
  ]}
 SetShowBrowseToAssetWarning {nodes [
   n1 = SetShowBrowseToAssetWarning!() -> (then, ShowBrowseToAssetWarning:bool=~) {is=title K2Node_FunctionEntry}
   > Set ShowBrowseToAssetWarning!(execute, ShowBrowseToAssetWarning=@n1) -> (then, Output_Get)
   > SaveGame_Save()
  ]}
 ToggleCommonTagsEnabled {nodes [
   ToggleCommonTagsEnabled() {is=title K2Node_FunctionEntry}
   > n1 = Set ShowingCommonTags!(execute, ShowingCommonTags=NOT Boolean(A=Get ShowingCommonTags())) -> (then, Output_Get)
   > Call On Show Common Tags Changed(Enabled=@n1)
   > SaveGame_Save()
  ]}
 WriteQuestPropertiesFromTemplate {nodes [
   n1 = WriteQuestPropertiesFromTemplate!() -> (then, Template To Use:Object Reference, Quest Object To Use:Object Reference) {is=title K2Node_FunctionEntry}
   > Branch(Condition=AND Boolean(A=IsValid(Object=@n1), B=IsValid(Object=@n1.Quest Object To Use)))
   > Set templateObject(templateObject:Object Reference=@n1) -> (Output_Get:Object Reference)
   > n2 = For Each Loop(Array:Array of Names=Break Quest Editor Common Data!(QuestEditorCommonData=Get CommonData()) -> (
     TemplatedPropertyNames_13_#7:Array of Names
     ActiveObject_28_#1:Object Reference
     QuestTemplateTypeName_39_#4
    )) -> (Array Element:Name)
   > n3 = Set currentPropertyName(currentPropertyName:Name=@n2) -> (Output_Get:Name=None)
   > n4 = Branch(Condition=Contains Item(TargetArray=Get TemplatedContainerTagProperties() -> (TemplatedContainerTagProperties:Array of Names), ItemToFind=@n3))
   n5 = Get currentPropertyName() -> (currentPropertyName:Name=None)
   > PrintString(InString=Append(A="Cast Success Current Property: ", B="To String (Name)"(InName=@n5)))
   > n6 = GetEditorProperty!(
    execute:Exec
    Object:Object Reference=Get Quest Object To Use() -> (Quest Object To Use:Object Reference)
    PropertyName:Name=Get currentPropertyName() -> (currentPropertyName=None)
   ) -> (then:Exec, PropertyValue:Gameplay Tag Container Structure, ReturnValue:bool)
   > n7 = GetEditorProperty(Object=Get templateObject() -> (templateObject:Object Reference), PropertyName=Get currentPropertyName() -> (currentPropertyName=None)) -> (PropertyValue:"Gameplay Tag Container Structure (by ref)")
   > n8 = Set gameplayTagsToWrite(gameplayTagsToWrite:Gameplay Tag Container Structure=@n6) -> (Output_Get:Gameplay Tag Container Structure)
   > AppendGameplayTagContainers!(execute, InOutTagContainer:"Gameplay Tag Container Structure (by ref)"=@n8, InTagContainer:"Gameplay Tag Container Structure (by ref)"=@n7) -> (then)
   > SetEditorProperty(
    Object=Get Quest Object To Use()
    PropertyName=Get currentPropertyName() -> (currentPropertyName=None)
    PropertyValue:Gameplay Tag Container Structure=Get gameplayTagsToWrite() -> (gameplayTagsToWrite:Gameplay Tag Container Structure)
   )
   n9 = Get currentPropertyName() -> (currentPropertyName=None)
   PrintString(execute=@n4.else, InString=Append(A="Cast Failed Current Property: ", B="To String (Name)"(InName=@n9)))
   > Execute Python Script!(
    execute
    PythonScript="value = template_object.get_editor_property(property_name)\r\nquest_object.set_editor_property(property_name, value)"
    template_object:Object Reference=Get templateObject()
    quest_object:Object Reference=Get Quest Object To Use()
    property_name:Name=Get currentPropertyName() -> (currentPropertyName=None)
   ) -> (then, ReturnValue) {inputs [template_object, quest_object, property_name]}
  ]}
]
variables [
 CanvasRoot {type=CanvasPanel, default=null}
 CommonData {
  type=QuestEditorCommonData
  default {
   activeObject=null
   currentSeasonStyle=""
   fileName=new_quest
   questTemplateTypeName=None
   widgetType=QuestType
   folder {path=/BattlePassS19/Items/QuestItems/ToolsQuests/}
   templatedPropertyNames []
  }
 }
 OnActiveObjectChanged {default=()}
 OnBannerMessage {default=()}
 OnCopiedObjectiveIndexSet {default=()}
 OnSavedFilePickerPathChanged {default=()}
 OnShowCommonTagsChanged {default=()}
 SavedFilePickerPath {type=string, default=""}
 ShowBrowseToAssetWarning {type=boolean, default=true}
 ShowingCommonTags {type=boolean, default=true}
 TemplatedContainerTagProperties {type="array<string>", default [GameplayTags]}
 Templates {type=DataTable, default=/Game/NeverCook/EditorUtilities/QuestTools/Data/QuestToolTemplates}
]
